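# Unlit, two-sided, alpha-composite Niagara sprite material for fragment FX on the Tactical Woodland Blue male glider.
# Builds a procedural slanted band from UVs, with no textures, using Skew for its tilt and Softness for its edge falloff.
# Multiplies the band by particle color and alpha, then feeds it as emissive into MF_FX_EmissiveAdjustments.
# A Color vector parameter is present but unconnected.
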Material /BRCosmetics/Gadgets/Assets/VinderTech_GliderChute/Glider_Tactical_Woodland_Blue_Male/FX/Materials/M_Glider_TacticalWoodlandBlue_Fragment
graph {
 properties {
  bUsedWithNiagaraSprites=true
  blendMode=BLEND_AlphaComposite
  pixelDepthOffsetMode=PDOM_Legacy
  shadingModel=MSM_Unlit
  twoSided=true
  usedShadingModels=Unlit
 }
 nodes [
  n1 = ParticleColor()
  n2 = TextureCoordinate()
  n3 = ScalarParameter() {defaultValue=0.3, parameterName=Skew, sliderMax=1}
  n4 = Multiply(A=Subtract(A=ComponentMask(None=@n2) {r=true}) {constB=0.5}, B=@n3) {constB=0.3}
  n5 = Multiply(A=ConstantBiasScale(None=ComponentMask(None=@n2) {g=true}) {bias=-0.5, scale=2}, B=Add(A=Divide(A=@n3)))
  n6 = Saturate(None=Abs(None=Add(A=@n4, B=@n5)))
  n7 = Saturate(None=Divide(A=OneMinus(None=@n6), B=ScalarParameter() {defaultValue=0.1, parameterName=Softness}))
  MaterialFunctionCall(Emissive=Multiply(A=@n1.RGB, B=Multiply(A=@n1.A, B=@n7))) {materialFunction=&MF_FX_EmissiveAdjustments}
  VectorParameter() {parameterName=Color, defaultValue=<rgba 0 0.6667 1 0.5>}
 ]
}
