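# Unlit, two-sided alpha-composite Niagara material for the blue Tactical Woodland glider's thruster exhaust.
# Sine-warped UVs pan a lines texture with per-tile offsets, thresholded into streaks that blend hue-shifted color over a soft background.
# Dynamic parameters drive time and brightness, depth fade softens intersections, and particle red scales a vertex-normal offset.
# A separate vector-noise branch builds two eroded streak layers into a vertex interpolator with no visible consumer.
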Material /BRCosmetics/Gadgets/Assets/VinderTech_GliderChute/Glider_Tactical_Woodland_Blue_Male/FX/Materials/M_Glider_TacticalWoodlandBlue_Thruster
graph {
 properties {
  bUsedWithNiagaraSprites=true
  blendMode=BLEND_AlphaComposite
  pixelDepthOffsetMode=PDOM_Legacy
  shadingModel=MSM_Unlit
  twoSided=true
  usedShadingModels=Unlit
 }
 nodes [
  n1 = VectorParameter() {parameterName=Background Color, defaultValue=<rgba 0 0.0967 0.2 0>}
  n2 = TextureCoordinate()
  n3 = Subtract(A=ComponentMask(None=@n2) {r=true}) {constB=0.5}
  n4 = ComponentMask(None=@n2) {g=true}
  n5 = Add(A=@n4, B=ScalarParameter() {defaultValue=0.4125, parameterName=UV Distortion Shift, sliderMax=1})
  n6 = Add(A=Clamp(None=Sine(None=@n5)) {maxDefault=0.25, minDefault=-0.25}) {constB=0.25}
  n7 = ScalarParameter() {defaultValue=0.625, parameterName=UV Distortion Intensity}
  n8 = Multiply(A=@n3, B=Add(B=Multiply(A=@n6, B=@n7)) {constA=1})
  n9 = Saturate(None=Add(A=@n8) {constB=0.5})
  n10 = Multiply(A=@n9, B=OneMinus(None=@n9))
  n11 = Saturate(None=Divide(A=@n10, B=ScalarParameter() {defaultValue=0.5, parameterName=Color Background Softness A}))
  n12 = Saturate(None=Divide(A=@n10, B=ScalarParameter() {defaultValue=0.02, parameterName=Color Background Softness B}) {constB=0.02})
  n13 = OneMinus(None=@n4)
  n14 = Multiply(A=Divide(A=Add(A=@n11, B=@n12)), B=Multiply(A=@n13, B=@n13))
  n15 = ConstantBiasScale(None=ComponentMask(None=TextureCoordinate()) {g=true}) {bias=-0.5, scale=2}
  n16 = VectorParameter() {parameterName="Lines Color && Hue Shift", defaultValue=<rgba 0 0.6667 1 0.11>}
  n17 = Multiply(A=OneMinus(None=Abs(None=@n15)), B=@n16.A)
  n18 = Multiply(A=@n8, B=ScalarParameter() {defaultValue=0.7, parameterName=UV Scale U})
  n19 = Add(A=Abs(None=@n3), B=@n4)
  n20 = ScalarParameter() {defaultValue=0.65, parameterName=UV Scale V}
  n21 = Panner(
   Coordinate=AppendVector(A=Add(A=@n18) {constB=0.5}, B=Multiply(A=@n19, B=@n20))
   Time=Add(A=Time(), B=DynamicParameter() {paramNames [Time Increment, Brightness Multiplier, - unused -, - unused -]}.Time Increment)
  ) {speedY=-1}
  n22 = Multiply(A=Floor(None=ComponentMask(None=@n21) {g=true})) {constB=0.135}
  n23 = Add(A=@n21, B=AppendVector(A=Frac(None=@n22), B=Constant()))
  n24 = TextureSample(UVs=@n23) {
   samplerType=SAMPLERTYPE_LinearGrayscale
   texture=/BRCosmetics/Gadgets/Assets/VinderTech_GliderChute/Glider_Tactical_Woodland_Blue_Male/FX/Textures/T_Glider_TacticalWoodlandBlue_Lines
  }
  n25 = ComponentMask(None=TextureCoordinate()) {g=true}
  n26 = Divide(A=OneMinus(None=@n25)) {constB=0.2}
  n27 = Multiply(A=@n25, B=@n25)
  n28 = Abs(None=@n8)
  n29 = Saturate(None=Divide(A=Multiply(A=@n27, B=@n28)) {constB=0.2})
  n30 = OneMinus(None=Multiply(A=Saturate(None=@n26), B=OneMinus(None=@n29)))
  n31 = ScalarParameter() {defaultValue=0.1, parameterName=Softness}
  n32 = Divide(A=Subtract(A=@n24.R, B=@n30), B=@n31)
  n33 = TextureSample(UVs=Multiply(A=Add(A=@n23) {constB=0.5}) {constB=2}) {
   samplerType=SAMPLERTYPE_LinearGrayscale
   texture=&T_Glider_TacticalWoodlandBlue_Lines
  }
  n34 = Saturate(None=Divide(A=Subtract(A=@n33.R, B=@n30), B=@n31))
  n35 = ScalarParameter() {defaultValue=0.0333, parameterName=Frequency Blend, sliderMax=1}
  n36 = Saturate(None=Add(A=Saturate(None=@n32), B=Multiply(A=@n34, B=@n35)))
  n37 = Multiply(
   A=LinearInterpolate(
    A=Multiply(A=@n1.RGB, B=@n14)
    B=MaterialFunctionCall(Hue Shift Percentage=@n17, Texture=@n16.RGB) {materialFunction=&HueShift}.Result
    Alpha=@n36
   )
   B=Add(A=DynamicParameter() {paramNames [Time Increment, Brightness Multiplier, - unused -, - unused -]}.Brightness Multiplier) {constB=0.5}
  )
  n38 = ComponentMask(None=TextureCoordinate()) {g=true}
  n39 = ParticleColor()
  n40 = DepthFade(Opacity=@n39.A, FadeDistance=ScalarParameter() {defaultValue=11, parameterName=Depth Fade Distance})
  MaterialFunctionCall(Emissive=Multiply(A=Divide(A=@n37, B=@n38), B=@n40)) {materialFunction=&MF_FX_EmissiveAdjustments}
  Multiply(A=@n36, B=@n40)
  Multiply(A=VertexNormalWS(), B=@n39.R) {constB=1.6}
  n41 = TextureCoordinate()
  n42 = Panner(Coordinate=Multiply(A=@n41, B=Constant2Vector() {g=1, r=4}) {constB=4}) {speedY=-2.618}
  n43 = VectorNoise(World Position=AppendVector(A=@n42, B=Constant()))
  n44 = Subtract(A=ComponentMask(None=@n43) {r=true}) {constB=0.5}
  n45 = Multiply(A=@n44, B=ScalarParameter() {defaultValue=0.2, parameterName=UV Distortion Random Shift})
  n46 = ComponentMask(None=@n41) {g=true}
  n47 = Clamp(None=MaterialFunctionCall(Value=Add(A=@n45, B=@n46), Period=Constant() {r=1}, -1 to 1=StaticBool() {value=true}, Sine Phase=StaticBool()) {materialFunction=&LinearSine}.Linear Sine) {maxDefault=0.25, minDefault=-0.25}
  n48 = ScalarParameter() {defaultValue=0.625, parameterName=UV Distortion Intensity}
  n49 = TextureCoordinate()
  n50 = Subtract(A=OneMinus(None=ComponentMask(None=@n49) {r=true})) {constB=0.5}
  n51 = Multiply(A=ComponentMask(None=@n43) {g=true}, B=Multiply(A=Multiply(A=@n47, B=@n48), B=Sign(None=@n50)))
  n52 = Multiply(A=Subtract(A=ComponentMask(None=@n43) {b=true}) {constB=0.5}, B=ScalarParameter() {defaultValue=1, parameterName=UV Side Shift Random})
  n53 = ScalarParameter() {defaultValue=0.6, parameterName=UV Gradient Random Shift}
  n54 = Add(A=@n42, B=AppendVector(A=Add(A=@n51, B=@n52), B=Multiply(A=@n44, B=@n53)))
  n55 = ConstantBiasScale(None=Frac(None=ComponentMask(None=@n54) {r=true})) {bias=-0.5, scale=2}
  n56 = Saturate(None=Divide(A=Abs(None=@n55)) {constB=0.2})
  n57 = ComponentMask(None=@n54) {g=true}
  n58 = Divide(A=ConstantBiasScale(None=Frac(None=@n57)) {bias=-0.5, scale=2}, B=OneMinus(None=@n53)) {constB=0.2}
  n59 = OneMinus(None=Saturate(None=Abs(None=@n58)))
  n60 = ScalarParameter() {defaultValue=1, parameterName=Fade Size}
  n61 = OneMinus(None=Saturate(None=Divide(A=@n59, B=@n60) {constB=0.2}))
  n62 = ComponentMask(None=TextureCoordinate()) {g=true}
  n63 = OneMinus(None=Abs(None=ConstantBiasScale(None=@n62) {bias=-0.5, scale=2}))
  n64 = OneMinus(None=Saturate(None=Divide(A=@n63) {constB=0.2}))
  n65 = TextureCoordinate()
  n66 = Panner(Coordinate=Multiply(A=@n65, B=Constant2Vector() {g=1, r=4}) {constB=4}) {speedY=-4.23}
  n67 = VectorNoise(World Position=AppendVector(A=@n66, B=Constant()))
  n68 = Subtract(A=ComponentMask(None=@n67) {r=true}) {constB=0.5}
  n69 = Multiply(A=@n68, B=ScalarParameter() {defaultValue=0.2, parameterName=UV Distortion Random Shift})
  n70 = ComponentMask(None=@n65) {g=true}
  n71 = Clamp(None=MaterialFunctionCall(Value=Add(A=@n69, B=@n70), Period=Constant() {r=1}, -1 to 1=StaticBool() {value=true}, Sine Phase=StaticBool()) {materialFunction=&LinearSine}.Linear Sine) {maxDefault=0.25, minDefault=-0.25}
  n72 = ScalarParameter() {defaultValue=0.625, parameterName=UV Distortion Intensity}
  n73 = TextureCoordinate()
  n74 = Subtract(A=OneMinus(None=ComponentMask(None=@n73) {r=true})) {constB=0.5}
  n75 = Multiply(A=ComponentMask(None=@n67) {g=true}, B=Multiply(A=Multiply(A=@n71, B=@n72), B=Sign(None=@n74)))
  n76 = Multiply(A=Subtract(A=ComponentMask(None=@n67) {b=true}) {constB=0.5}, B=ScalarParameter() {defaultValue=1, parameterName=UV Side Shift Random})
  n77 = ScalarParameter() {defaultValue=0.6, parameterName=UV Gradient Random Shift}
  n78 = Add(A=@n66, B=AppendVector(A=Add(A=@n75, B=@n76), B=Multiply(A=@n68, B=@n77)))
  n79 = ConstantBiasScale(None=Frac(None=ComponentMask(None=@n78) {r=true})) {bias=-0.5, scale=2}
  n80 = Saturate(None=Divide(A=Abs(None=@n79)) {constB=0.2})
  n81 = ComponentMask(None=@n78) {g=true}
  n82 = Divide(A=ConstantBiasScale(None=Frac(None=@n81)) {bias=-0.5, scale=2}, B=OneMinus(None=@n77)) {constB=0.2}
  n83 = OneMinus(None=Saturate(None=Abs(None=@n82)))
  n84 = ScalarParameter() {defaultValue=1, parameterName=Fade Size}
  n85 = OneMinus(None=Saturate(None=Divide(A=@n83, B=@n84) {constB=0.2}))
  n86 = ComponentMask(None=TextureCoordinate()) {g=true}
  n87 = OneMinus(None=Abs(None=ConstantBiasScale(None=@n86) {bias=-0.5, scale=2}))
  n88 = OneMinus(None=Saturate(None=Divide(A=@n87) {constB=0.2}))
  n89 = Add(
   A=MaterialFunctionCall!(Gradient=OneMinus(None=@n56), Erosion Amount=Max(A=@n61, B=@n64)) {materialFunction=/BRCosmetics/Athena/Effects/Particles/Skydiving/NS_Contrail_DarkAndLight/Functions/MF_ErosionNormalized}.Result
   B=MaterialFunctionCall(Gradient=OneMinus(None=@n80), Erosion Amount=Max(A=@n85, B=@n88)) {materialFunction=&MF_ErosionNormalized}.Result
  )
  n90 = TextureCoordinate()
  n91 = Multiply(A=OneMinus(None=ComponentMask(None=@n90) {g=true})) {constB=0.01}
  VertexInterpolator(VS=Multiply(A=Add(A=@n89, B=@n91), B=VectorParameter() {parameterName=Color, defaultValue=<rgba 0 23.7656 33 0>}.RGB))
 ]
}
