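# Glider material for the Tactical Woodland Blue male cosmetic that adds an animated numeric counter screen to the shared glider material functions.
# An input number is split into hundreds, tens and ones, each picked from a flipbook atlas in three UV columns with kerning masks.
# Panned static, a sweep line, scanlines and a frame vignette modulate the counter color.
# The FX mask isolates the screen in emissive and base color, plus an optional negative sign.
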Material /BRCosmetics/Gadgets/Assets/VinderTech_GliderChute/Glider_Tactical_Woodland_Blue_Male/Materials/M_FN_Glider_TacticalWoodlandBlue
graph {
 properties {
  bUseMaterialAttributes=true
  bUsedWithMeshDeformer=true
  bUsedWithSkeletalMesh=true
  pixelDepthOffsetMode=PDOM_Legacy
  usedShadingModels=Default Lit
 }
 nodes [
  n1 = Constant() {r=0.5}
  n2 = LinearInterpolate(
   A=TextureCoordinate()
   B=TextureCoordinate() {coordinateIndex=1}
   Alpha=ScalarParameter() {group=Counter, parameterName=CounterUseSecondUVChannel}
  )
  n3 = Constant2Vector() {g=0.5, r=0.5}
  n4 = ScalarParameter() {group=Counter, parameterName=Counter_Rotation}
  n5 = ScalarParameter() {group=Counter, parameterName=U_pos}
  n6 = ScalarParameter() {group=Counter, parameterName=V_pos}
  n7 = ScalarParameter() {defaultValue=1, group=Counter, parameterName=CounterScaleU}
  n8 = PreSkinnedLocalBounds()
  n9 = Multiply(A=VertexInterpolator(VS=PreSkinnedPosition()).PS, B=ScalarParameter() {defaultValue=0.5, group=Counter, parameterName=SkewAmount})
  n10 = Multiply(
   A=ComponentMask(None=Subtract(A=@n8.Extents, B=@n9)) {b=true}
   B=ScalarParameter() {defaultValue=0.01, group=Counter, parameterName=SkewScale}
  )
  n11 = MaterialFunctionCall(
   UVs=Add(
    A=MaterialFunctionCall(UVs=@n2, Rotation Center=@n3, "Rotation Angle (0-1)"=@n4) {materialFunction=&CustomRotator}.Rotated Values
    B=AppendVector(A=@n5, B=@n6)
   )
   Texture Scale=MaterialFunctionCall(X=Multiply(A=@n7, B=@n10), Y=ScalarParameter() {defaultValue=1, group=Counter, parameterName=CounterScaleV}) {materialFunction=&MakeFloat2}.Result
  ) {materialFunction=&ScaleUVsByCenter}
  n12 = Frac(None=@n11.UVs)
  n13 = Constant2Vector() {g=1, r=3}
  n14 = Add(A=Frac(None=Multiply(A=@n12, B=@n13))) {constB=-0.25}
  n15 = Sine(None=ComponentMask(None=@n14) {r=true})
  n16 = ScalarParameter() {defaultValue=0.02, group=Counter - Kerning, parameterName=KerningEdgeMaskSize}
  n17 = Ceil(None=Subtract(A=ConstantBiasScale(None=@n15), B=@n16) {constB=0.1})
  n18 = Sine(None=ComponentMask(None=@n14) {g=true})
  n19 = Ceil(None=Subtract(A=ConstantBiasScale(None=@n18), B=@n16) {constB=0.1})
  n20 = Clamp(None=Multiply(A=Multiply(A=@n17, B=@n19)) {constB=5})
  n21 = MaterialFunctionCall(Float2=Multiply(A=@n11.UVs, B=@n13)) {materialFunction=&BreakOutFloat2Components}
  n22 = Multiply(A=Frac(None=@n21.G)) {constB=0.1}
  n23 = Abs(None=ScalarParameter() {defaultValue=2603, desc=input number, group=Counter, parameterName=Counter_InputNumber})
  n24 = Constant() {r=100}
  n25 = Max(A=Floor(None=Divide(A=@n23, B=@n24) {constB=10})) {constB=0}
  n26 = NamedRerouteUsage()
  n27 = Round(None=Fmod(A=@n23, B=@n24))
  n28 = Constant() {r=10}
  n29 = Divide(A=@n27, B=@n28) {constB=10}
  n30 = NamedRerouteUsage()
  n31 = Multiply(A=Max(A=Floor(None=@n29)) {constB=0}, B=@n30)
  n32 = Fmod(A=@n27, B=@n28)
  n33 = Multiply(A=Max(A=Round(None=@n32)) {constB=0}, B=NamedRerouteUsage())
  n34 = Divide(
   A=Add(A=Multiply(A=@n25, B=@n26), B=Add(A=@n31, B=@n33))
   B=ScalarParameter() {defaultValue=10, desc=Number of Total Digits in Texture, group=Counter, parameterName=numColumns}
  ) {desc="multiply by how many digits in the texture.\r\nThis gives us an offset."}
  n35 = TextureSampleParameter2D(UVs=MaterialFunctionCall(X=@n21.R, Y=Add(A=@n22, B=@n34)) {materialFunction=&MakeFloat2}.Result) {
   group=Counter
   parameterName=Counter_AtlasTexture
   texture=/Game/Accessories/FORT_Backpacks/Backpack_F_MED_Cotton_Candy/Textures/T_NumberFlipBook
  }
  n36 = ComponentMask(None=Multiply(A=@n11.0-1 mask, B=Multiply(A=@n20, B=@n35.RGB))) {r=true}
  n37 = Multiply(A=Multiply(A=@n1, B=Saturate(None=@n36)), B=ScalarParameter() {defaultValue=5, group=Counter, parameterName=FaceIntensity})
  n38 = Multiply(A=ComponentMask(None=NamedRerouteUsage()) {g=true}) {constB=0.75}
  n39 = Multiply(A=Time()) {constB=-0.2}
  n40 = Power(Base=Frac(None=Add(A=@n38, B=@n39))) {constExponent=20}
  n41 = ScalarParameter() {defaultValue=0.2, group=Sweep Line, parameterName=Sweep Add}
  n42 = NamedRerouteUsage()
  n43 = Panner(Coordinate=@n42) {speedX=5, speedY=0.1}
  n44 = Multiply(
   A=TextureSample(UVs=Panner(Coordinate=@n42) {speedX=6, speedY=-0.05}) {samplerType=SAMPLERTYPE_Masks, texture=/Game/Missions/Primary/RepairTheShelter/Textures/Cinematic_Screen_Static}.G
   B=TextureSample(UVs=Panner(Coordinate=@n42) {speedX=-3, speedY=-0.03}) {samplerType=SAMPLERTYPE_Masks, texture=&Cinematic_Screen_Static}.R
  )
  n45 = Add(
   A=Multiply(A=TextureSample(UVs=@n43) {samplerType=SAMPLERTYPE_Masks, texture=&Cinematic_Screen_Static}.B) {constB=0.1}
   B=@n44
  )
  n46 = ScalarParameter() {defaultValue=6, group=Static, parameterName=Static Scale}
  n47 = Add(
   A=Add(A=Multiply(A=@n45, B=@n46), B=Multiply(A=@n40, B=@n41) {constB=0.2})
   B=ScalarParameter() {defaultValue=0.2, group=Background, parameterName=Background}
  )
  Add(A=Multiply(A=@n47, B=@n37), B=Multiply(A=@n37) {constB=0.1}) {constB=0.5}
  AppendVector(A=Multiply(A=@n40, B=ScalarParameter() {defaultValue=0.005, group=Sweep Line, parameterName=Sweep UV Distort}) {constB=0.02}, B=Constant())
  AppendVector(A=Multiply(A=@n44, B=ScalarParameter() {defaultValue=0.07, group=Static, parameterName=Static UV Distort}), B=Constant())
  n48 = ComponentMask(None=@n11.UVs) {r=true}
  n49 = Multiply(A=@n48) {constB=3}
  Fmod(A=@n49, B=Constant() {r=1})
  LinearInterpolate(
   A=MaterialFunctionCall!(ScanLine_Tiling_1, Scanline_PannerSpeed_1, ScanLine_Tiling, Scanline_PannerSpeed, UVs) {materialFunction=/Game/Weapons/WeaponSkins/MaterialFunction/ScanLines}.Result
   B=@n1
   Alpha=MaterialFunctionCall!(Blend Range=ScalarParameter() {defaultValue=25, parameterName=r}, Start Offset=ScalarParameter() {defaultValue=-100, parameterName=d}) {materialFunction=/Engine/Functions/Engine_MaterialFunctions03/Particles/Distance_Blend}.Result
  )
  LinearInterpolate(
   A=LinearInterpolate(
    A=VectorParameter() {parameterName=color3, defaultValue=<rgba 0 0.131 0.9375 0>}.RGB
    B=VectorParameter() {parameterName=color2, defaultValue=<rgba 0 0.974 0.1767 0>}.RGB
    Alpha=@n30
   )
   B=VectorParameter() {parameterName=color1, defaultValue=<rgba 0.6719 0 0 0>}.RGB
   Alpha=@n26
  )
  n50 = MaterialFunctionCall!(In=MaterialFunctionCall() {materialFunction=/Game/Gadgets/Assets/VinderTech_GliderChute/Materials/Functions/MF_Glider_TextureInputs}.Result) {desc=Build-A-Brella, materialFunction=/Game/Gadgets/Assets/VinderTech_GliderChute/Materials/Functions/MF_Glider_BAB}
  n51 = GetMaterialAttributes(None=MaterialFunctionCall!(
    Pre-Skinned World Normal
    Gmap
    Customization Mask
    IsCharacter
    IsStatsB
    Material Attributes=MaterialFunctionCall!(Material=@n50.Result) {materialFunction=/Game/Gadgets/Assets/VinderTech_GliderChute/Materials/Functions/MF_Glider_Emissive}.Result
   ) {materialFunction=/Game/Weapons/WeaponSkins/MaterialFunction/Weapon_SkinFXFunction}.Result) {attributeGetTypes [#1, #2]}
  n52 = TextureSampleParameter2D() {
   parameterName=FX_Mask
   texture=/BRCosmetics/Accessories/FORT_Backpacks/Backpack_M_MED_Tactical_Woodland_Blue/Textures/T_M_MED_Tactical_Woodland_Blue_Backpack_FX
  }
  n53 = Add(
   A=LinearInterpolate(A=@n51.Emissive Color, Alpha=@n52.G) {constB=0}
   B=Multiply(
    A=LinearInterpolate(B=@n51.Emissive Color, Alpha=@n52.G) {desc=IsolateScreenEmissive}
    B=ScalarParameter() {defaultValue=0.7, group=Counter, parameterName=AdjustOtherScreenBrightness}
   )
  )
  n54 = Multiply(
   A=VectorParameter() {group=Counter, parameterName=Counter Color, defaultValue=<rgba 0.1595 0.1913 0.8229 0>}.RGB
   B=ScalarParameter() {defaultValue=1, group=Counter, parameterName=CounterBrightness}
  )
  n55 = Multiply(A=@n54, B=ScalarParameter() {defaultValue=2, group=Counter, parameterName=NegativeSym_Mult})
  n56 = Multiply(A=@n52.R, B=ScalarParameter() {parameterName=IsNegative})
  n57 = Multiply(A=NamedRerouteUsage()) {constB=0.5}
  n58 = Sine(None=ComponentMask(None=@n57) {r=true})
  n59 = Sine(None=ComponentMask(None=@n57) {g=true})
  n60 = Power(Base=OneMinus(None=Multiply(A=@n58, B=@n59)), Exp=ScalarParameter() {defaultValue=0.8, group=Background, parameterName=Frame Exponent})
  n61 = NamedRerouteUsage()
  n62 = ConstantBiasScale(None=Frac(None=Panner(Coordinate=@n61) {speedY=-1})) {bias=-0.5, scale=2}
  n63 = Multiply(
   A=ComponentMask(None=Multiply(A=@n62, B=@n62)) {g=true}
   B=ScalarParameter() {defaultValue=0.5, group=Background, parameterName=Scanline Strength}
  ) {constB=0.1}
  n64 = Multiply(A=Multiply(A=OneMinus(None=@n60), B=Add(A=@n63)), B=Add(A=@n47, B=@n37)) {constA=1}
  n65 = LinearInterpolate(B=Add(A=LinearInterpolate(B=@n55, Alpha=@n56), B=Multiply(A=@n54, B=@n64)), Alpha=@n52.G) {desc=IsolateScreen}
  MaterialFunctionCall(In=SetMaterialAttributes!(
    MaterialAttributes=@n51.MaterialAttributes
    Emissive Color=Add(A=@n53, B=@n65)
    Base Color=LinearInterpolate(A=@n51.Base Color, Alpha=@n52.G) {constB=0}
   ) {attributeSetTypes [#1, #2]}) {materialFunction=&MF_TechArtOverrides_accessories}
  n66 = MaterialFunctionCall(Float2=TextureCoordinate()) {materialFunction=&BreakOutFloat2Components}
  n67 = MaterialFunctionCall(X=@n66.G, Y=@n66.G) {materialFunction=&MakeFloat2}
  n68 = ComponentMask(None=VectorParameter() {group=Sweep Line, parameterName=Scale_Sweep, defaultValue=<rgba 1 1 0 0>}.RGB) {g=true, r=true}
  n69 = VectorParameter() {group=Sweep Line, parameterName=Offset_Sweep}
  NamedRerouteDeclaration(None=Add(A=Multiply(A=@n67.Result, B=@n68), B=ComponentMask(None=@n69.RGB) {g=true, r=true})) {name=TexCoord, nodeColor=<rgba 1 0 0.4453 1>}
  n70 = Floor(None=@n49)
  n71 = Subtract(B=Saturate(None=Subtract(A=@n70)))
  NamedRerouteDeclaration(None=Saturate(None=OneMinus(None=@n71))) {name=OnesPlace, nodeColor=<rgba 0.6172 0 1 1>}
  n72 = Saturate(None=@n70)
  NamedRerouteDeclaration(None=Saturate(None=OneMinus(None=@n72))) {name=HundredsPlace, nodeColor=<rgba 1 0 0.8906 1>}
  n73 = Add(A=NamedRerouteUsage(), B=NamedRerouteUsage())
  NamedRerouteDeclaration(None=Saturate(None=Subtract(B=@n73))) {name=TensPlace, nodeColor=<rgba 0 1 0.4141 1>}
  n74 = Sine(None=MaterialFunctionCall(
    TimestampA=ScalarParameter() {parameterName=TimeStart}
    TransitionLength_In=ScalarParameter() {defaultValue=1, group=Counter, parameterName=FlashLength}
    Transition In Only?=StaticBool() {value=true}
   ) {materialFunction=&MF_Timestamp_InOut_01}.Linear) {period=2}
  Reroute(None=Add(A=Multiply(A=@n74) {constB=20}))
  n75 = Sine(None=Divide(A=Time()) {constB=10})
  Reroute(None=Multiply(A=ScalarParameter() {defaultValue=4, group=Background, parameterName=EffectPulseMultiplier}, B=Abs(None=@n75)) {constB=4})
  Round(None=@n48)
  TextureCoordinate() {vTiling=15}
  TextureCoordinate()
  TextureCoordinate() {coordinateIndex=1}
 ]
}
